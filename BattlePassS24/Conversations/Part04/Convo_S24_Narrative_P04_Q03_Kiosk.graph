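# Conversation database for the Season 24 Part 04 Q03 quest kiosk, a maze puzzle where the player moves a circle to an X.
# Each grid cell is a Branch entry point with four arrow choices; valid moves link to neighboring cells, while obstacle or invalid moves show an error and go Back.
# Cell-numbered metadata tags each result.
# Start leads to Branch.05; Up from Branch.06 grants access and fires the quest prop success side effect.
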ConversationDatabase /BattlePassS24/Conversations/Part04/Convo_S24_Narrative_P04_Q03_Kiosk
entry_points [
 {tag=Conversation.Entry.Season24.Kiosk.P04.Q03.Start, destinations [#1]}
 {tag=Conversation.Entry.Season24.Kiosk.P04.Q03.Branch.06, destinations [#2]}
 {tag=Conversation.Entry.Season24.Kiosk.P04.Q03.Branch.03, destinations [#3]}
 {tag=Conversation.Entry.Season24.Kiosk.01.Branch.05, destinations [#4]}
 {tag=Conversation.Entry.Season24.Kiosk.P04.Q03.Branch.05, destinations [#5]}
 {tag=Conversation.Entry.Season24.Kiosk.P04.Q03.Branch.08, destinations [#6]}
 {tag=Conversation.Entry.Season24.Kiosk.P04.Q03.Branch.14, destinations [#7]}
 {tag=Conversation.Entry.Season24.Kiosk.P04.Q03.Branch.11, destinations [#8]}
 {tag=Conversation.Entry.Season24.Kiosk.P04.Q03.Branch.04, destinations [#9]}
 {tag=Conversation.Entry.Season24.Kiosk.P04.Q03.Branch.13, destinations [#10]}
 {tag=Conversation.Entry.Season24.Kiosk.P04.Q03.Branch.12, destinations [#11]}
 {tag=Conversation.Entry.Season24.Kiosk.P04.Q03.Branch.15, destinations [#12]}
]
nodes [
 ConversationEntryPointNode_11 {
  id=#12
  class=ConversationEntryPointNode
  outputs [#13, #14, #15, #16]
  properties {entryTag {tagName=Conversation.Entry.Season24.Kiosk.P04.Q03.Branch.15}}
 }
 Speech_C_57 {
  id=#16
  class=Speech_C
  outputs [#17]
  properties {
   softTaskIcon=/Game/Athena/UI/InGame/Creative/PropertyEditor/Icon_Picker/Textures_V2/64x64/T_UI_IconLibrary_RightArrow_64
   generalConfig {defaultMessage=Invalid move - try again, defaultMetadataParameters [15 {value=Error}]}
  }
  sub_nodes [ConversationChoiceNode_18 {class=ConversationChoiceNode, properties {defaultChoiceDisplayText=Go Right}}]
 }
 Back_C_30 {id=#17, class=Back_C}
 Speech_C_56 {
  id=#15
  class=Speech_C
  outputs [#18]
  properties {
   softTaskIcon=/Game/Athena/UI/InGame/Creative/PropertyEditor/Icon_Picker/Textures_V2/64x64/T_UI_IconLibrary_DownArrow_64
   generalConfig {defaultMessage=Invalid move - try again, defaultMetadataParameters [15 {value=Error}]}
  }
  sub_nodes [ConversationChoiceNode_16 {class=ConversationChoiceNode, properties {defaultChoiceDisplayText=Go Down}}]
 }
 Back_C_35 {id=#18, class=Back_C}
 Speech_C_55 {
  id=#14
  class=Speech_C
  outputs [#19]
  properties {
   softTaskIcon=/Game/Athena/UI/InGame/Creative/PropertyEditor/Icon_Picker/Textures_V2/64x64/T_UI_IconLibrary_UpArrow_64
   generalConfig {defaultMessage=Obstacle - try again, defaultMetadataParameters [15 {value=Error}]}
  }
  sub_nodes [ConversationChoiceNode_14 {class=ConversationChoiceNode, properties {defaultChoiceDisplayText=Go Up}}]
 }
 Back_C_21 {id=#19, class=Back_C}
 Speech_C_50 {
  id=#13
  class=Speech_C
  outputs [#20]
  properties {
   softTaskIcon=/Game/Athena/UI/InGame/Creative/PropertyEditor/Icon_Picker/Textures_V2/64x64/T_UI_IconLibrary_LeftArrow_64
   generalConfig {defaultMessage=Move the circle to the X, defaultMetadataParameters [15 {value=Left}]}
  }
  sub_nodes [ConversationChoiceNode_13 {class=ConversationChoiceNode, properties {defaultChoiceDisplayText=Go Left}}]
 }
 ConversationLinkNode_18 {
  id=#20
  class=ConversationLinkNode
  properties {remoteEntryTag {tagName=Conversation.Entry.Season24.Kiosk.P04.Q03.Branch.14}}
 }
 ConversationEntryPointNode_7 {
  id=#11
  class=ConversationEntryPointNode
  outputs [#21, #22, #23, #24]
  properties {entryTag {tagName=Conversation.Entry.Season24.Kiosk.P04.Q03.Branch.12}}
 }
 Speech_C_45 {
  id=#24
  class=Speech_C
  outputs [#25]
  properties {
   softTaskIcon=&T_UI_IconLibrary_RightArrow_64
   generalConfig {defaultMessage=Move the circle to the X, defaultMetadataParameters [12 {value=Right}]}
  }
  sub_nodes [ConversationChoiceNode_29 {class=ConversationChoiceNode, properties {defaultChoiceDisplayText=Go Right}}]
 }
 ConversationLinkNode_8 {
  id=#25
  class=ConversationLinkNode
  properties {remoteEntryTag {tagName=Conversation.Entry.Season24.Kiosk.P04.Q03.Branch.13}}
 }
 Speech_C_44 {
  id=#23
  class=Speech_C
  outputs [#26]
  properties {
   softTaskIcon=&T_UI_IconLibrary_DownArrow_64
   generalConfig {defaultMessage=Invalid move - try again, defaultMetadataParameters [12 {value=Error}]}
  }
  sub_nodes [ConversationChoiceNode_27 {class=ConversationChoiceNode, properties {defaultChoiceDisplayText=Go Down}}]
 }
 Back_C_27 {id=#26, class=Back_C}
 Speech_C_43 {
  id=#22
  class=Speech_C
  outputs [#27]
  properties {
   softTaskIcon=&T_UI_IconLibrary_UpArrow_64
   generalConfig {defaultMessage=Obstacle - try again, defaultMetadataParameters [12 {value=Error}]}
  }
  sub_nodes [ConversationChoiceNode_26 {class=ConversationChoiceNode, properties {defaultChoiceDisplayText=Go Up}}]
 }
 Back_C_13 {id=#27, class=Back_C}
 Speech_C_42 {
  id=#21
  class=Speech_C
  outputs [#28]
  properties {
   softTaskIcon=&T_UI_IconLibrary_LeftArrow_64
   generalConfig {defaultMessage=Move the circle to the X, defaultMetadataParameters [12 {value=Left}]}
  }
  sub_nodes [ConversationChoiceNode_24 {class=ConversationChoiceNode, properties {defaultChoiceDisplayText=Go Left}}]
 }
 ConversationLinkNode_33 {
  id=#28
  class=ConversationLinkNode
  properties {remoteEntryTag {tagName=Conversation.Entry.Season24.Kiosk.P04.Q03.Branch.11}}
 }
 ConversationEntryPointNode_6 {
  id=#10
  class=ConversationEntryPointNode
  outputs [#29, #30, #31, #32]
  properties {entryTag {tagName=Conversation.Entry.Season24.Kiosk.P04.Q03.Branch.13}}
 }
 Speech_C_28 {
  id=#32
  class=Speech_C
  outputs [#33]
  properties {
   softTaskIcon=&T_UI_IconLibrary_RightArrow_64
   generalConfig {defaultMessage=Move the circle to the X, defaultMetadataParameters [13 {value=Right}]}
  }
  sub_nodes [ConversationChoiceNode_38 {class=ConversationChoiceNode, properties {defaultChoiceDisplayText=Go Right}}]
 }
 ConversationLinkNode_15 {
  id=#33
  class=ConversationLinkNode
  properties {remoteEntryTag {tagName=Conversation.Entry.Season24.Kiosk.P04.Q03.Branch.14}}
 }
 Speech_C_26 {
  id=#31
  class=Speech_C
  outputs [#34]
  properties {
   softTaskIcon=&T_UI_IconLibrary_DownArrow_64
   generalConfig {defaultMessage=Invalid move - try again, defaultMetadataParameters [13 {value=Error}]}
  }
  sub_nodes [ConversationChoiceNode_40 {class=ConversationChoiceNode, properties {defaultChoiceDisplayText=Go Down}}]
 }
 Back_C_14 {id=#34, class=Back_C}
 Speech_C_29 {
  id=#30
  class=Speech_C
  outputs [#35]
  properties {
   softTaskIcon=&T_UI_IconLibrary_UpArrow_64
   generalConfig {defaultMessage=Move the circle to the X, defaultMetadataParameters [13 {value=Up}]}
  }
  sub_nodes [ConversationChoiceNode_37 {class=ConversationChoiceNode, properties {defaultChoiceDisplayText=Go Up}}]
 }
 ConversationLinkNode_22 {
  id=#35
  class=ConversationLinkNode
  properties {remoteEntryTag {tagName=Conversation.Entry.Season24.Kiosk.P04.Q03.Branch.08}}
 }
 Speech_C_27 {
  id=#29
  class=Speech_C
  outputs [#36]
  properties {
   softTaskIcon=&T_UI_IconLibrary_LeftArrow_64
   generalConfig {defaultMessage=Move the circle to the X, defaultMetadataParameters [13 {value=Left}]}
  }
  sub_nodes [ConversationChoiceNode_39 {class=ConversationChoiceNode, properties {defaultChoiceDisplayText=Go Left}}]
 }
 ConversationLinkNode_11 {
  id=#36
  class=ConversationLinkNode
  properties {remoteEntryTag {tagName=Conversation.Entry.Season24.Kiosk.P04.Q03.Branch.12}}
 }
 ConversationEntryPointNode_3 {
  id=#9
  class=ConversationEntryPointNode
  outputs [#37, #38, #39, #40]
  properties {entryTag {tagName=Conversation.Entry.Season24.Kiosk.P04.Q03.Branch.04}}
 }
 Speech_C_4 {
  id=#40
  class=Speech_C
  outputs [#41]
  properties {
   softTaskIcon=&T_UI_IconLibrary_RightArrow_64
   generalConfig {defaultMessage=Move the circle to the X, defaultMetadataParameters [4 {value=Right}]}
  }
  sub_nodes [ConversationChoiceNode_8 {class=ConversationChoiceNode, properties {defaultChoiceDisplayText=Go Right}}]
 }
 ConversationLinkNode_3 {
  id=#41
  class=ConversationLinkNode
  properties {remoteEntryTag {tagName=Conversation.Entry.Season24.Kiosk.P04.Q03.Branch.05}}
 }
 Speech_C_3 {
  id=#39
  class=Speech_C
  outputs [#42]
  properties {
   softTaskIcon=&T_UI_IconLibrary_DownArrow_64
   generalConfig {defaultMessage=Obstacle - try again, defaultMetadataParameters [4 {value=Error}]}
  }
  sub_nodes [ConversationChoiceNode_9 {class=ConversationChoiceNode, properties {defaultChoiceDisplayText=Go Down}}]
 }
 Back_C_8 {id=#42, class=Back_C}
 Speech_C_2 {
  id=#38
  class=Speech_C
  outputs [#43]
  properties {
   softTaskIcon=&T_UI_IconLibrary_UpArrow_64
   generalConfig {defaultMessage=Invalid move - try again, defaultMetadataParameters [4 {value=Error}]}
  }
  sub_nodes [ConversationChoiceNode_10 {class=ConversationChoiceNode, properties {defaultChoiceDisplayText=Go Up}}]
 }
 Back_C_5 {id=#43, class=Back_C}
 Speech_C_1 {
  id=#37
  class=Speech_C
  outputs [#44]
  properties {
   softTaskIcon=&T_UI_IconLibrary_LeftArrow_64
   generalConfig {defaultMessage=Move the circle to the X, defaultMetadataParameters [4 {value=Left}]}
  }
  sub_nodes [ConversationChoiceNode_12 {class=ConversationChoiceNode, properties {defaultChoiceDisplayText=Go Left}}]
 }
 ConversationLinkNode_2 {
  id=#44
  class=ConversationLinkNode
  properties {remoteEntryTag {tagName=Conversation.Entry.Season24.Kiosk.P04.Q03.Branch.03}}
 }
 ConversationEntryPointNode_12 {
  id=#8
  class=ConversationEntryPointNode
  outputs [#45, #46, #47, #48]
  properties {entryTag {tagName=Conversation.Entry.Season24.Kiosk.P04.Q03.Branch.11}}
 }
 Speech_C_49 {
  id=#48
  class=Speech_C
  outputs [#49]
  properties {
   softTaskIcon=&T_UI_IconLibrary_RightArrow_64
   generalConfig {defaultMessage=Move the circle to the X, defaultMetadataParameters [11 {value=Right}]}
  }
  sub_nodes [ConversationChoiceNode_47 {class=ConversationChoiceNode, properties {defaultChoiceDisplayText=Go Right}}]
 }
 ConversationLinkNode_27 {
  id=#49
  class=ConversationLinkNode
  properties {remoteEntryTag {tagName=Conversation.Entry.Season24.Kiosk.P04.Q03.Branch.12}}
 }
 Speech_C_48 {
  id=#47
  class=Speech_C
  outputs [#50]
  properties {
   softTaskIcon=&T_UI_IconLibrary_DownArrow_64
   generalConfig {defaultMessage=Invalid move - try again, defaultMetadataParameters [11 {value=Error}]}
  }
  sub_nodes [ConversationChoiceNode_46 {class=ConversationChoiceNode, properties {defaultChoiceDisplayText=Go Down}}]
 }
 Back_C_19 {id=#50, class=Back_C}
 Speech_C_47 {
  id=#46
  class=Speech_C
  outputs [#51]
  properties {
   softTaskIcon=&T_UI_IconLibrary_UpArrow_64
   generalConfig {defaultMessage=Move the circle to the X, defaultMetadataParameters [11 {value=Up}]}
  }
  sub_nodes [ConversationChoiceNode_45 {class=ConversationChoiceNode, properties {defaultChoiceDisplayText=Go Up}}]
 }
 ConversationLinkNode_6 {
  id=#51
  class=ConversationLinkNode
  properties {remoteEntryTag {tagName=Conversation.Entry.Season24.Kiosk.P04.Q03.Branch.06}}
 }
 Speech_C_46 {
  id=#45
  class=Speech_C
  outputs [#52]
  properties {
   softTaskIcon=&T_UI_IconLibrary_LeftArrow_64
   generalConfig {defaultMessage=Invalid move - try again, defaultMetadataParameters [11 {value=Error}]}
  }
  sub_nodes [ConversationChoiceNode_36 {class=ConversationChoiceNode, properties {defaultChoiceDisplayText=Go Left}}]
 }
 Back_C_20 {id=#52, class=Back_C}
 ConversationEntryPointNode_16 {
  id=#7
  class=ConversationEntryPointNode
  outputs [#53, #54, #55, #56]
  properties {entryTag {tagName=Conversation.Entry.Season24.Kiosk.P04.Q03.Branch.14}}
 }
 Speech_C_25 {
  id=#56
  class=Speech_C
  outputs [#57]
  properties {
   softTaskIcon=&T_UI_IconLibrary_RightArrow_64
   generalConfig {defaultMessage=Move the circle to the X, defaultMetadataParameters [14 {value=Right}]}
  }
  sub_nodes [ConversationChoiceNode_60 {class=ConversationChoiceNode, properties {defaultChoiceDisplayText=Go Right}}]
 }
 ConversationLinkNode_16 {
  id=#57
  class=ConversationLinkNode
  properties {remoteEntryTag {tagName=Conversation.Entry.Season24.Kiosk.P04.Q03.Branch.15}}
 }
 Speech_C_24 {
  id=#55
  class=Speech_C
  outputs [#58]
  properties {
   softTaskIcon=&T_UI_IconLibrary_DownArrow_64
   generalConfig {defaultMessage=Invalid move - try again, defaultMetadataParameters [14 {value=Error}]}
  }
  sub_nodes [ConversationChoiceNode_59 {class=ConversationChoiceNode, properties {defaultChoiceDisplayText=Go Down}}]
 }
 Back_C_17 {id=#58, class=Back_C}
 Speech_C_23 {
  id=#54
  class=Speech_C
  outputs [#59]
  properties {
   softTaskIcon=&T_UI_IconLibrary_UpArrow_64
   generalConfig {defaultMessage=Obstacle - try again, defaultMetadataParameters [14 {value=Error}]}
  }
  sub_nodes [ConversationChoiceNode_58 {class=ConversationChoiceNode, properties {defaultChoiceDisplayText=Go Up}}]
 }
 Back_C_15 {id=#59, class=Back_C}
 Speech_C_22 {
  id=#53
  class=Speech_C
  outputs [#60]
  properties {
   softTaskIcon=&T_UI_IconLibrary_LeftArrow_64
   generalConfig {defaultMessage=Move the circle to the X, defaultMetadataParameters [14 {value=Left}]}
  }
  sub_nodes [ConversationChoiceNode_48 {class=ConversationChoiceNode, properties {defaultChoiceDisplayText=Go Left}}]
 }
 ConversationLinkNode_17 {
  id=#60
  class=ConversationLinkNode
  properties {remoteEntryTag {tagName=Conversation.Entry.Season24.Kiosk.P04.Q03.Branch.13}}
 }
 ConversationEntryPointNode_0 {
  id=#6
  class=ConversationEntryPointNode
  outputs [#61, #62, #63, #64]
  properties {entryTag {tagName=Conversation.Entry.Season24.Kiosk.P04.Q03.Branch.08}}
 }
 Speech_C_21 {
  id=#64
  class=Speech_C
  outputs [#65]
  properties {
   softTaskIcon=&T_UI_IconLibrary_RightArrow_64
   generalConfig {defaultMessage=Obstacle - try again, defaultMetadataParameters [8 {value=Error}]}
  }
  sub_nodes [ConversationChoiceNode_52 {class=ConversationChoiceNode, properties {defaultChoiceDisplayText=Go Right}}]
 }
 Back_C_12 {id=#65, class=Back_C}
 Speech_C_0 {
  id=#63
  class=Speech_C
  outputs [#66]
  properties {
   softTaskIcon=&T_UI_IconLibrary_DownArrow_64
   generalConfig {defaultMessage=Move the circle to the X, defaultMetadataParameters [8 {value=Down}]}
  }
  sub_nodes [ConversationChoiceNode_1 {class=ConversationChoiceNode, properties {defaultChoiceDisplayText=Go Down}}]
 }
 ConversationLinkNode_26 {
  id=#66
  class=ConversationLinkNode
  properties {remoteEntryTag {tagName=Conversation.Entry.Season24.Kiosk.P04.Q03.Branch.13}}
 }
 Speech_C_19 {
  id=#62
  class=Speech_C
  outputs [#67]
  properties {
   softTaskIcon=&T_UI_IconLibrary_UpArrow_64
   generalConfig {defaultMessage=Move the circle to the X, defaultMetadataParameters [8 {value=Up}]}
  }
  sub_nodes [ConversationChoiceNode_50 {class=ConversationChoiceNode, properties {defaultChoiceDisplayText=Go Up}}]
 }
 ConversationLinkNode_23 {
  id=#67
  class=ConversationLinkNode
  properties {remoteEntryTag {tagName=Conversation.Entry.Season24.Kiosk.P04.Q03.Branch.03}}
 }
 Speech_C_10 {
  id=#61
  class=Speech_C
  outputs [#68]
  properties {
   softTaskIcon=&T_UI_IconLibrary_LeftArrow_64
   generalConfig {defaultMessage=Obstacle - try again, defaultMetadataParameters [8 {value=Error}]}
  }
  sub_nodes [ConversationChoiceNode_49 {class=ConversationChoiceNode, properties {defaultChoiceDisplayText=Go Left}}]
 }
 Back_C_25 {id=#68, class=Back_C}
 ConversationEntryPointNode_9 {
  id=#5
  class=ConversationEntryPointNode
  outputs [#69, #70, #71, #72]
  properties {entryTag {tagName=Conversation.Entry.Season24.Kiosk.P04.Q03.Branch.05}}
 }
 Speech_C_16 {
  id=#72
  class=Speech_C
  outputs [#73]
  properties {
   softTaskIcon=&T_UI_IconLibrary_RightArrow_64
   generalConfig {defaultMessage=Invalid move - try again, defaultMetadataParameters [5 {value=Error}]}
  }
  sub_nodes [ConversationChoiceNode_17 {class=ConversationChoiceNode, properties {defaultChoiceDisplayText=Go Right}}]
 }
 Back_C_4 {id=#73, class=Back_C}
 Speech_C_15 {
  id=#71
  class=Speech_C
  outputs [#74]
  properties {
   softTaskIcon=&T_UI_IconLibrary_DownArrow_64
   generalConfig {defaultMessage=Obstacle - try again, defaultMetadataParameters [5 {value=Error}]}
  }
  sub_nodes [ConversationChoiceNode_7 {class=ConversationChoiceNode, properties {defaultChoiceDisplayText=Go Down}}]
 }
 Back_C_9 {id=#74, class=Back_C}
 Speech_C_17 {
  id=#70
  class=Speech_C
  outputs [#75]
  properties {
   softTaskIcon=&T_UI_IconLibrary_UpArrow_64
   generalConfig {defaultMessage=Invalid move - try again, defaultMetadataParameters [5 {value=Error}]}
  }
  sub_nodes [ConversationChoiceNode_34 {class=ConversationChoiceNode, properties {defaultChoiceDisplayText=Go Up}}]
 }
 Back_C_3 {id=#75, class=Back_C}
 Speech_C_9 {
  id=#69
  class=Speech_C
  outputs [#76]
  properties {
   softTaskIcon=&T_UI_IconLibrary_LeftArrow_64
   generalConfig {defaultMessage=Move the circle to the X, defaultMetadataParameters [5 {value=Left}]}
  }
  sub_nodes [ConversationChoiceNode_28 {class=ConversationChoiceNode, properties {defaultChoiceDisplayText=Go Left}}]
 }
 ConversationLinkNode_12 {
  id=#76
  class=ConversationLinkNode
  properties {remoteEntryTag {tagName=Conversation.Entry.Season24.Kiosk.P04.Q03.Branch.04}}
 }
 ConversationEntryPointNode_8 {
  id=#4
  class=ConversationEntryPointNode
  outputs []
  properties {entryTag {tagName=Conversation.Entry.Season24.Kiosk.01.Branch.05}}
 }
 ConversationEntryPointNode_2 {
  id=#3
  class=ConversationEntryPointNode
  outputs [#77, #78, #79, #80]
  properties {entryTag {tagName=Conversation.Entry.Season24.Kiosk.P04.Q03.Branch.03}}
 }
 Speech_C_5 {
  id=#80
  class=Speech_C
  outputs [#81]
  properties {
   softTaskIcon=&T_UI_IconLibrary_RightArrow_64
   generalConfig {defaultMessage=Move the circle to the X, defaultMetadataParameters [3 {value=Right}]}
  }
  sub_nodes [ConversationChoiceNode_6 {class=ConversationChoiceNode, properties {defaultChoiceDisplayText=Go Right}}]
 }
 ConversationLinkNode_1 {
  id=#81
  class=ConversationLinkNode
  properties {remoteEntryTag {tagName=Conversation.Entry.Season24.Kiosk.P04.Q03.Branch.04}}
 }
 Speech_C_6 {
  id=#79
  class=Speech_C
  outputs [#82]
  properties {
   softTaskIcon=&T_UI_IconLibrary_DownArrow_64
   generalConfig {defaultMessage=Move the circle to the X, defaultMetadataParameters [3 {value=Down}]}
  }
  sub_nodes [ConversationChoiceNode_5 {class=ConversationChoiceNode, properties {defaultChoiceDisplayText=Go Down}}]
 }
 ConversationLinkNode_4 {
  id=#82
  class=ConversationLinkNode
  properties {remoteEntryTag {tagName=Conversation.Entry.Season24.Kiosk.P04.Q03.Branch.08}}
 }
 Speech_C_7 {
  id=#78
  class=Speech_C
  outputs [#83]
  properties {
   softTaskIcon=&T_UI_IconLibrary_UpArrow_64
   generalConfig {defaultMessage=Invalid move - try again, defaultMetadataParameters [3 {value=Error}]}
  }
  sub_nodes [ConversationChoiceNode_4 {class=ConversationChoiceNode, properties {defaultChoiceDisplayText=Go Up}}]
 }
 Back_C_1 {id=#83, class=Back_C}
 Speech_C_8 {
  id=#77
  class=Speech_C
  outputs [#84]
  properties {
   softTaskIcon=&T_UI_IconLibrary_LeftArrow_64
   generalConfig {defaultMessage=Obstacle - try again, defaultMetadataParameters [3 {value=Error}]}
  }
  sub_nodes [ConversationChoiceNode_0 {class=ConversationChoiceNode, properties {defaultChoiceDisplayText=Go Left}}]
 }
 Back_C_6 {id=#84, class=Back_C}
 ConversationEntryPointNode_19 {
  id=#2
  class=ConversationEntryPointNode
  outputs [#85, #86, #87, #88]
  properties {entryTag {tagName=Conversation.Entry.Season24.Kiosk.P04.Q03.Branch.06}}
 }
 Speech_C_39 {
  id=#88
  class=Speech_C
  outputs [#89]
  properties {
   softTaskIcon=&T_UI_IconLibrary_RightArrow_64
   generalConfig {defaultMessage=Obstacle - try again, defaultMetadataParameters [6 {value=Error}]}
  }
  sub_nodes [ConversationChoiceNode_25 {class=ConversationChoiceNode, properties {defaultChoiceDisplayText=Go Right}}]
 }
 Back_C_10 {id=#89, class=Back_C}
 Speech_C_41 {
  id=#87
  class=Speech_C
  outputs [#90]
  properties {
   softTaskIcon=&T_UI_IconLibrary_DownArrow_64
   generalConfig {defaultMessage=Move the circle to the X, defaultMetadataParameters [6 {value=Down}]}
  }
  sub_nodes [ConversationChoiceNode_19 {class=ConversationChoiceNode, properties {defaultChoiceDisplayText=Go Down}}]
 }
 ConversationLinkNode_0 {
  id=#90
  class=ConversationLinkNode
  properties {remoteEntryTag {tagName=Conversation.Entry.Season24.Kiosk.P04.Q03.Branch.11}}
 }
 Speech_C_38 {
  id=#86
  class=Speech_C
  outputs []
  properties {
   softTaskIcon=&T_UI_IconLibrary_UpArrow_64
   generalConfig {defaultMessage="Access Granted!", defaultMetadataParameters [6 {value=Win}]}
  }
  sub_nodes [
   ConversationChoiceNode_33 {class=ConversationChoiceNode, properties {defaultChoiceDisplayText=Go Up}}
   ConvoSideEffect_QuestPropActor_ConvoSuccess_C_1 {class=ConvoSideEffect_QuestPropActor_ConvoSuccess_C}
  ]
 }
 Speech_C_40 {
  id=#85
  class=Speech_C
  outputs [#91]
  properties {
   softTaskIcon=&T_UI_IconLibrary_LeftArrow_64
   generalConfig {defaultMessage=Invalid move - try again, defaultMetadataParameters [6 {value=Error}]}
  }
  sub_nodes [ConversationChoiceNode_20 {class=ConversationChoiceNode, properties {defaultChoiceDisplayText=Go Left}}]
 }
 Back_C_0 {id=#91, class=Back_C}
 ConversationEntryPointNode_4 {
  id=#1
  class=ConversationEntryPointNode
  outputs [#92]
  properties {entryTag {tagName=Conversation.Entry.Season24.Kiosk.P04.Q03.Start}}
 }
 Speech_C_11 {
  id=#92
  class=Speech_C
  outputs [#93]
  properties {
   softTaskIcon=/Game/UI/Foundation/Textures/Icons/Replay/T-Icon-Replay-Plus-L
   generalConfig {defaultMessage=Move the circle to the X, defaultMetadataParameters [1 {value=Default}]}
  }
 }
 ConversationLinkNode_9 {
  id=#93
  class=ConversationLinkNode
  properties {remoteEntryTag {tagName=Conversation.Entry.Season24.Kiosk.P04.Q03.Branch.05}}
 }
]
